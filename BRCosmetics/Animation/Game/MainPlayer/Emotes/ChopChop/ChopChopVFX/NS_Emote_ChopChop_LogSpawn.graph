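# Niagara system for the log spawn effect of the ChopChop emote on the main player.
# It combines three emitters named for sparkles, airglow and smoke, with SystemState as the only listed system module.
# User variables expose front-end and front-end preview flags plus a front-end scale, apparently to adapt the effect for menu display.
# Emitter modules and renderers are not listed.
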NiagaraSystem /BRCosmetics/Animation/Game/MainPlayer/Emotes/ChopChop/ChopChopVFX/NS_Emote_ChopChop_LogSpawn
dependencies {usedDataInterfaces [], usedDynamicInputs [], usedModules [&SystemState]}
emitters {
 AirglowLWE002 {rendererClasses [], renderers []}
 SmokeLWE002 {rendererClasses [], renderers []}
 SparklesLWE002 {rendererClasses [], renderers []}
}
summary {
 systemName=NS_Emote_ChopChop_LogSpawn
 emitters [
  {emitterName=SparklesLWE002, rendererClasses []}
  {emitterName=AirglowLWE002, rendererClasses []}
  {emitterName=SmokeLWE002, rendererClasses []}
 ]
 userVariables [
  User.bIsFrontEnd {type=NiagaraBool}
  User.bIsFrontEndPreview {type=NiagaraBool}
  User.FrontEndScale {type=NiagaraFloat, defaultValue=0.4, description=""}
 ]
}
